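# Sound cue for the looping turn and aim movement sound of the Ostrich vehicle.
# A single looping upper-body turn-aim wave feeds a continuous modulator node at the root, which allows runtime modulation of the loop.
# It plays at reduced volume through the Athena player foley sound class, using a close-range Ostrich movement attenuation asset.
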SoundCue /Game/Athena/Sounds/Vehicles/Ostrich/Ostrich_Angular_Movement_Cue
root=ModulatorContinuous_5
nodes [
 {id=ModulatorContinuous_5, children [WavePlayer_7]}
 {id=WavePlayer_7, properties {bLooping=true, soundWaveAssetPtr=&/Veh_Ostrich_TurnAim_Upper_01}}
]
properties {
 attenuationSettings=/Game/Sounds/Attenuation/Fort_Ostrich_Movement_Close_Att
 maxDistance=6512
 soundClassObject=&Fort_Player_Foley_Athena
 volumeMultiplier=0.5
}
