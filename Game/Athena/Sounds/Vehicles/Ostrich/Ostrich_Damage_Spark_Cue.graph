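# Sound cue for electrical damage sparks on the Ostrich vehicle, reusing two mono spark waves from the electric ceiling trap.
# An equal-weight random pick feeds the first two slots of a quality-level switch, leaving the other two empty and therefore silent.
# A modulator varies pitch, then an enveloper fades volume out over a short window.
# It uses the Vehicles sound class and close Ostrich movement attenuation.
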SoundCue /Game/Athena/Sounds/Vehicles/Ostrich/Ostrich_Damage_Spark_Cue
root=Enveloper_0
nodes [
 {
  id=Enveloper_0
  children [Modulator_0]
  properties {pitchCurve {editorCurveData {keys [{time=0}]}}, volumeCurve {editorCurveData {keys [<at 0 1>,<at 0.75 0>]}}}
 }
 {id=Modulator_0, children [QualityLevel_0], properties {pitchMax=1.1, pitchMin=0.8}}
 {id=QualityLevel_0, children [Random_1,Random_1,null,null]}
 {id=Random_1, children [WavePlayer_0,WavePlayer_3], properties {weights [1,1]}}
 {id=WavePlayer_0, properties {soundWaveAssetPtr=/Game/Sounds/Fort_Traps/Electric_Trap/CeilingTrap_Electric_Spark_mono_001}}
 {id=WavePlayer_3, properties {soundWaveAssetPtr=/Game/Sounds/Fort_Traps/Electric_Trap/CeilingTrap_Electric_Spark_mono_002}}
]
properties {
 attenuationSettings=/Game/Sounds/Attenuation/Fort_Ostrich_Movement_Close_Att
 maxDistance=6512
 soundClassObject=&Fort_Vehicles
 volumeMultiplier=0.6
}
