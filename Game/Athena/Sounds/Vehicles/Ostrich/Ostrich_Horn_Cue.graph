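# Sound cue for the Ostrich vehicle's horn in Athena, playing a single horn wave with no randomization or mixing.
# It routes through the vehicles sound class at reduced volume and uses the Ostrich close-movement attenuation asset.
# A concurrency override allows one instance per owner and blocks new plays while one is active, so repeated honks do not stack.
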SoundCue /Game/Athena/Sounds/Vehicles/Ostrich/Ostrich_Horn_Cue
root=WavePlayer_0
nodes [{id=WavePlayer_0, properties {soundWaveAssetPtr=&/Veh_Ostrich_Horn_01}}]
properties {
 attenuationSettings=/Game/Sounds/Attenuation/Fort_Ostrich_Movement_Close_Att
 bOverrideConcurrency=true
 maxDistance=6512
 soundClassObject=&Fort_Vehicles
 volumeMultiplier=0.5
 concurrencyOverrides {bLimitToOwner=true, maxCount=1, resolutionRule=PreventNew}
}
