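# Sound cue for Ostrich vehicle linear movement, a looping upper-body turn/aim wave played as close-attenuated foley.
# A continuous modulator driven by the DriverAngle parameter maps pitch upward and volume inversely, from full down to half, as the parameter rises.
# An Enveloper at the root applies a volume curve that drops from full to a low level over the first second.
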SoundCue /Game/Athena/Sounds/Vehicles/Ostrich/Ostrich_Linear_Movement_Cue
root=Enveloper_0
nodes [
 {
  id=Enveloper_0
  children [ModulatorContinuous_5]
  properties {pitchCurve {editorCurveData {keys [{time=0}]}}, volumeCurve {editorCurveData {keys [<at 0 1>,<at 1 0.2>]}}}
 }
 {
  id=ModulatorContinuous_5
  children [WavePlayer_7]
  properties {
   pitchModulationParams {default=0, maxOutput=1.2, minOutput=1, parameterName=DriverAngle}
   volumeModulationParams {default=0, maxOutput=0.5, minOutput=1, parameterName=DriverAngle}
  }
 }
 {id=WavePlayer_7, properties {bLooping=true, soundWaveAssetPtr=&/Veh_Ostrich_TurnAim_Upper_01}}
]
properties {
 attenuationSettings=/Game/Sounds/Attenuation/Fort_Ostrich_Movement_Close_Att
 maxDistance=6512
 soundClassObject=&Fort_Player_Foley_Athena
 volumeMultiplier=0.35
}
