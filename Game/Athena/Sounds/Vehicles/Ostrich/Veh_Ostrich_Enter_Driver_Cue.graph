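# Sound cue for the Ostrich vehicle's driver entry, classed as third-person enemy player foley with close Ostrich movement attenuation.
# A Mixer layers the immediate enter sound with two quieter, delayed Ostrich FS (likely footstep) hits.
# Each hit waits a fixed delay, staggering them within the entry, and picks randomly from the same four equally weighted variants.
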SoundCue /Game/Athena/Sounds/Vehicles/Ostrich/Veh_Ostrich_Enter_Driver_Cue
root=Mixer_0
nodes [
 {id=Delay_0, children [Random_0], properties {delayMax=1.25, delayMin=1.25}}
 {id=Delay_1, children [Random_1], properties {delayMax=1.9, delayMin=1.9}}
 {id=Mixer_0, children [WavePlayer_0,Delay_0,Delay_1], properties {inputVolume [1,0.5,0.5]}}
 {id=Random_0, children [WavePlayer_1,WavePlayer_2,WavePlayer_3,WavePlayer_4], properties {weights [1,1,1,1]}}
 {id=Random_1, children [WavePlayer_5,WavePlayer_6,WavePlayer_7,WavePlayer_8], properties {weights [1,1,1,1]}}
 {id=WavePlayer_0, properties {soundWaveAssetPtr=&/Veh_Ostrich_Enter_Driver__01}}
 {id=WavePlayer_1, properties {soundWaveAssetPtr=&/Veh_Ostrich_FS_01}}
 {id=WavePlayer_2, properties {soundWaveAssetPtr=&/Veh_Ostrich_FS_02}}
 {id=WavePlayer_3, properties {soundWaveAssetPtr=&/Veh_Ostrich_FS_03}}
 {id=WavePlayer_4, properties {soundWaveAssetPtr=&/Veh_Ostrich_FS_04}}
 {id=WavePlayer_5, properties {soundWaveAssetPtr=&Veh_Ostrich_FS_01}}
 {id=WavePlayer_6, properties {soundWaveAssetPtr=&Veh_Ostrich_FS_02}}
 {id=WavePlayer_7, properties {soundWaveAssetPtr=&Veh_Ostrich_FS_03}}
 {id=WavePlayer_8, properties {soundWaveAssetPtr=&Veh_Ostrich_FS_04}}
]
properties {
 attenuationSettings=/Game/Sounds/Attenuation/Fort_Ostrich_Movement_Close_Att
 maxDistance=6512
 soundClassObject=&Fort_Player_Foley_3P_Enemy_Athena
 volumeMultiplier=0.6
}
