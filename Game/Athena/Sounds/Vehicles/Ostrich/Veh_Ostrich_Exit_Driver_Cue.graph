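# Sound cue for the Ostrich vehicle driver exit, played as enemy third-person player foley.
# A fixed delay node holds back a single exit-lower wave player, with no randomization or mixing.
# It uses the close Ostrich movement attenuation and plays at reduced volume.
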SoundCue /Game/Athena/Sounds/Vehicles/Ostrich/Veh_Ostrich_Exit_Driver_Cue
root=Delay_0
nodes [
 {id=Delay_0, children [WavePlayer_0], properties {delayMax=0.85, delayMin=0.85}}
 {id=WavePlayer_0, properties {soundWaveAssetPtr=&/Veh_Ostrich_Exit_Lower}}
]
properties {
 attenuationSettings=/Game/Sounds/Attenuation/Fort_Ostrich_Movement_Close_Att
 maxDistance=6512
 soundClassObject=&Fort_Player_Foley_3P_Enemy_Athena
 volumeMultiplier=0.35
}
