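# Sound cue for the gunner exiting the Ostrich vehicle in Athena.
# A single wave player plays the Veh_Ostrich_Exit_Lower sound at reduced volume, with no randomization or layering.
# It uses the Ostrich close-movement attenuation and the third-person enemy player foley sound class.
# A concurrency override allows one instance at a time and blocks new plays while it is active.
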SoundCue /Game/Athena/Sounds/Vehicles/Ostrich/Veh_Ostrich_Exit_Gunner_Cue
root=WavePlayer_0
nodes [{id=WavePlayer_0, properties {soundWaveAssetPtr=&/Veh_Ostrich_Exit_Lower}}]
properties {
 attenuationSettings=/Game/Sounds/Attenuation/Fort_Ostrich_Movement_Close_Att
 maxDistance=6512
 soundClassObject=&Fort_Player_Foley_3P_Enemy_Athena
 volumeMultiplier=0.35
 concurrencyOverrides {maxCount=1, resolutionRule=PreventNew}
}
