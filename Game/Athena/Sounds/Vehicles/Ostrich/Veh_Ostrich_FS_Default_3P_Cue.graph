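# Sound cue for the Ostrich vehicle's default third-person footsteps, classed as player foley and using a distant movement attenuation.
# A Random node picks one of six equally weighted footstep waves, and a Modulator varies its pitch and volume before a Mixer with one empty second input.
# The cue boosts overall volume, and its concurrency overrides specify one instance per owner, stopping the oldest.
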SoundCue /Game/Athena/Sounds/Vehicles/Ostrich/Veh_Ostrich_FS_Default_3P_Cue
root=Mixer_0
nodes [
 {id=Mixer_0, children [Modulator_0,null], properties {inputVolume [1,1]}}
 {id=Modulator_0, children [Random_2], properties {pitchMin=0.9, volumeMin=0.9}}
 {
  id=Random_2
  children [WavePlayer_20, WavePlayer_21, WavePlayer_23, WavePlayer_24, WavePlayer_22, WavePlayer_25]
  properties {weights [1,1,1,1,1,1]}
 }
 {id=WavePlayer_20, properties {soundWaveAssetPtr=&/Veh_Ostrich_FS_01}}
 {id=WavePlayer_21, properties {soundWaveAssetPtr=&/Veh_Ostrich_FS_02}}
 {id=WavePlayer_22, properties {soundWaveAssetPtr=&/Veh_Ostrich_FS_03}}
 {id=WavePlayer_23, properties {soundWaveAssetPtr=&/Veh_Ostrich_FS_04}}
 {id=WavePlayer_24, properties {soundWaveAssetPtr=&/Veh_Ostrich_FS_05}}
 {id=WavePlayer_25, properties {soundWaveAssetPtr=&/Veh_Ostrich_FS_06}}
]
properties {
 attenuationSettings=/Game/Sounds/Attenuation/Fort_Ostrich_Movement_Distant_Att
 maxDistance=15000
 soundClassObject=&Fort_Player_Foley_Athena
 volumeMultiplier=1.75
 concurrencyOverrides {bLimitToOwner=true, maxCount=1, resolutionRule=StopOldest}
}
