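# Sound cue for the Ostrich vehicle's default medium landing footstep, classed as third-person enemy player foley.
# A Random node picks one of two equally weighted land wave variations, which passes through a Modulator at the root.
# Spatialization comes from a shared distant Ostrich movement attenuation asset, and volume is halved.
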SoundCue /Game/Athena/Sounds/Vehicles/Ostrich/Veh_Ostrich_FS_Default_Land_Med_Bank_Cue
root=Modulator_1
nodes [
 {id=Modulator_1, children [Random_0]}
 {id=Random_0, children [WavePlayer_0,WavePlayer_1], properties {weights [1,1]}}
 {id=WavePlayer_0, properties {soundWaveAssetPtr=&/Veh_Ostrich_FS_Land_Med_01}}
 {id=WavePlayer_1, properties {soundWaveAssetPtr=&/Veh_Ostrich_FS_Land_Med_02}}
]
properties {
 attenuationSettings=/Game/Sounds/Attenuation/Fort_Ostrich_Movement_Distant_Att
 maxDistance=15000
 soundClassObject=&Fort_Player_Foley_3P_Enemy_Athena
 volumeMultiplier=0.5
 attenuationOverrides {bAttenuate=false, falloffDistance=30000, attenuationShapeExtents {x=0}}
 concurrencyOverrides {bLimitToOwner=true, maxCount=1, resolutionRule=StopOldest}
 concurrencySet [null]
}
