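# Sound cue for the default-surface, medium landing footstep of the Ostrich vehicle, which is currently silent because it has no root or sound nodes.
# It keeps only playback settings: the enemy third-person foley sound class and a global footstep concurrency set.
# Attenuation override is enabled with a long falloff distance, alongside a referenced player landing attenuation asset.
# Concurrency override values specify one owner-limited instance that stops the oldest.
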SoundCue /Game/Athena/Sounds/Vehicles/Ostrich/Veh_Ostrich_FS_Default_Land_Med_Cue
root=null
nodes []
properties {
 attenuationSettings=/Game/Sounds/Attenuation/Fort_Player_Land_3P_Athena_Att
 bOverrideAttenuation=true
 maxDistance=30000
 soundClassObject=&Fort_Player_Foley_3P_Enemy_Athena
 attenuationOverrides {falloffDistance=30000, attenuationShapeExtents {x=0}}
 concurrencyOverrides {bLimitToOwner=true, maxCount=1, resolutionRule=StopOldest}
 concurrencySet [&Athena_Footstep_Global_Conc]
}
