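# Sound cue for the servo layer of Ostrich vehicle footsteps in Athena, routed to the player foley sound class.
# A random node picks one of four equally weighted servo waves, which passes through slight pitch and volume modulation and a short fixed delay at the root.
# It uses the Ostrich distant movement attenuation and the local footstep concurrency set.
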SoundCue /Game/Athena/Sounds/Vehicles/Ostrich/Veh_Ostrich_FS_Servo_Cue
root=Delay_1
nodes [
 {id=Delay_1, children [Modulator_2], properties {delayMax=0.3, delayMin=0.3}}
 {id=Modulator_2, children [Random_2], properties {pitchMin=0.9, volumeMin=0.8}}
 {id=Random_2, children [WavePlayer_19, WavePlayer_20, WavePlayer_23, WavePlayer_24], properties {weights [1,1,1,1]}}
 {id=WavePlayer_19, properties {soundWaveAssetPtr=&/Veh_Ostrich_FS_Servo_01}}
 {id=WavePlayer_20, properties {soundWaveAssetPtr=&/Veh_Ostrich_FS_Servo_02}}
 {id=WavePlayer_23, properties {soundWaveAssetPtr=&/Veh_Ostrich_FS_Servo_05}}
 {id=WavePlayer_24, properties {soundWaveAssetPtr=&/Veh_Ostrich_FS_Servo_06}}
]
properties {
 attenuationSettings=/Game/Sounds/Attenuation/Fort_Ostrich_Movement_Distant_Att
 maxDistance=15000
 soundClassObject=&Fort_Player_Foley_Athena
 volumeMultiplier=1
 concurrencyOverrides {bLimitToOwner=true, maxCount=1, resolutionRule=StopOldest}
 concurrencySet [&Athena_Footstep_Local_Conc]
}
