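# Sound cue for the Ostrich mech vehicle that plays a small servo foley sound, apparently at the start of a footstep.
# A Random node picks one of six equally weighted procedural servo waves and feeds a Modulator that varies pitch and volume on each play.
# It uses the Ostrich distant movement attenuation, the Athena player foley sound class and the local footstep concurrency set, with a volume boost.
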SoundCue /Game/Athena/Sounds/Vehicles/Ostrich/Veh_Ostrich_FS_Start_Cue
root=Modulator_0
nodes [
 {id=Modulator_0, children [Random_2], properties {pitchMin=0.9, volumeMin=0.8}}
 {
  id=Random_2
  children [WavePlayer_20, WavePlayer_21, WavePlayer_22, WavePlayer_23, WavePlayer_24, WavePlayer_28]
  properties {weights [1,1,1,1,1,1]}
 }
 {
  id=WavePlayer_20
  properties {soundWaveAssetPtr=/Game/Athena/Sounds/Vehicles/Ostrich/ProceduralFoley/Mech_Vehicle_Prodedural_Servo_SetB_Small_10}
 }
 {
  id=WavePlayer_21
  properties {soundWaveAssetPtr=/Game/Athena/Sounds/Vehicles/Ostrich/ProceduralFoley/Mech_Vehicle_Prodedural_Servo_SetB_Small_01}
 }
 {
  id=WavePlayer_22
  properties {soundWaveAssetPtr=/Game/Athena/Sounds/Vehicles/Ostrich/ProceduralFoley/Mech_Vehicle_Prodedural_Servo_SetB_Small_02}
 }
 {
  id=WavePlayer_23
  properties {soundWaveAssetPtr=/Game/Athena/Sounds/Vehicles/Ostrich/ProceduralFoley/Mech_Vehicle_Prodedural_Servo_SetB_Small_03}
 }
 {
  id=WavePlayer_24
  properties {soundWaveAssetPtr=/Game/Athena/Sounds/Vehicles/Ostrich/ProceduralFoley/Mech_Vehicle_Prodedural_Servo_SetB_Small_04}
 }
 {
  id=WavePlayer_28
  properties {soundWaveAssetPtr=/Game/Athena/Sounds/Vehicles/Ostrich/ProceduralFoley/Mech_Vehicle_Prodedural_Servo_SetB_Small_08}
 }
]
properties {
 attenuationSettings=/Game/Sounds/Attenuation/Fort_Ostrich_Movement_Distant_Att
 maxDistance=15000
 soundClassObject=&Fort_Player_Foley_Athena
 volumeMultiplier=2.5
 concurrencyOverrides {bLimitToOwner=true, maxCount=1, resolutionRule=StopOldest}
 concurrencySet [&Athena_Footstep_Local_Conc]
}
