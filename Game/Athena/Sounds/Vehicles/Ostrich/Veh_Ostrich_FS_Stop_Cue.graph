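# Sound cue for a footstep stop sound on the Ostrich mech vehicle, built from procedural servo foley.
# A Random node picks equally among six small servo waves, then a Modulator varies pitch and volume per play.
# It uses Ostrich distant-movement attenuation, the player foley sound class and a boosted volume.
# It references a local footstep concurrency set and carries an override allowing one instance per owner, stopping the oldest.
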SoundCue /Game/Athena/Sounds/Vehicles/Ostrich/Veh_Ostrich_FS_Stop_Cue
root=Modulator_2
nodes [
 {id=Modulator_2, children [Random_0], properties {pitchMin=0.9, volumeMin=0.8}}
 {id=Random_0, children [WavePlayer_1,WavePlayer_2,WavePlayer_3,WavePlayer_4,WavePlayer_5,WavePlayer_9], properties {weights [1,1,1,1,1,1]}}
 {
  id=WavePlayer_1
  properties {soundWaveAssetPtr=/Game/Athena/Sounds/Vehicles/Ostrich/ProceduralFoley/Mech_Vehicle_Prodedural_Servo_SetB_Small_10}
 }
 {
  id=WavePlayer_2
  properties {soundWaveAssetPtr=/Game/Athena/Sounds/Vehicles/Ostrich/ProceduralFoley/Mech_Vehicle_Prodedural_Servo_SetB_Small_01}
 }
 {
  id=WavePlayer_3
  properties {soundWaveAssetPtr=/Game/Athena/Sounds/Vehicles/Ostrich/ProceduralFoley/Mech_Vehicle_Prodedural_Servo_SetB_Small_02}
 }
 {
  id=WavePlayer_4
  properties {soundWaveAssetPtr=/Game/Athena/Sounds/Vehicles/Ostrich/ProceduralFoley/Mech_Vehicle_Prodedural_Servo_SetB_Small_03}
 }
 {
  id=WavePlayer_5
  properties {soundWaveAssetPtr=/Game/Athena/Sounds/Vehicles/Ostrich/ProceduralFoley/Mech_Vehicle_Prodedural_Servo_SetB_Small_04}
 }
 {
  id=WavePlayer_9
  properties {soundWaveAssetPtr=/Game/Athena/Sounds/Vehicles/Ostrich/ProceduralFoley/Mech_Vehicle_Prodedural_Servo_SetB_Small_08}
 }
]
properties {
 attenuationSettings=/Game/Sounds/Attenuation/Fort_Ostrich_Movement_Distant_Att
 maxDistance=15000
 soundClassObject=&Fort_Player_Foley_Athena
 volumeMultiplier=2
 concurrencyOverrides {bLimitToOwner=true, maxCount=1, resolutionRule=StopOldest}
 concurrencySet [&Athena_Footstep_Local_Conc]
}
