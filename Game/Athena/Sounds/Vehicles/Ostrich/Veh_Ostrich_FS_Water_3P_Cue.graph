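# Sound cue for third-person water footsteps of the Ostrich vehicle.
# A Random node picks one of six equally weighted water footstep waves, a Modulator varies pitch and volume, and a Mixer whose second input is empty passes it to the output.
# It uses distant Ostrich movement attenuation and the player foley sound class, with concurrency allowing one instance per owner that stops the oldest.
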SoundCue /Game/Athena/Sounds/Vehicles/Ostrich/Veh_Ostrich_FS_Water_3P_Cue
root=Mixer_0
nodes [
 {id=Mixer_0, children [Modulator_0,null], properties {inputVolume [1,1]}}
 {id=Modulator_0, children [Random_0], properties {pitchMin=0.9, volumeMin=0.9}}
 {
  id=Random_0
  children [WavePlayer_0, WavePlayer_7, WavePlayer_8, WavePlayer_9, WavePlayer_16, WavePlayer_17]
  properties {weights [1,1,1,1,1,1]}
 }
 {id=WavePlayer_0, properties {soundWaveAssetPtr=&/Veh_Ostrich_FS_Water_01}}
 {id=WavePlayer_16, properties {soundWaveAssetPtr=&/Veh_Ostrich_FS_Water_05}}
 {id=WavePlayer_17, properties {soundWaveAssetPtr=&/Veh_Ostrich_FS_Water_06}}
 {id=WavePlayer_7, properties {soundWaveAssetPtr=&/Veh_Ostrich_FS_Water_02}}
 {id=WavePlayer_8, properties {soundWaveAssetPtr=&/Veh_Ostrich_FS_Water_03}}
 {id=WavePlayer_9, properties {soundWaveAssetPtr=&/Veh_Ostrich_FS_Water_04}}
]
properties {
 attenuationSettings=/Game/Sounds/Attenuation/Fort_Ostrich_Movement_Distant_Att
 maxDistance=15000
 soundClassObject=&Fort_Player_Foley_Athena
 volumeMultiplier=1.75
 concurrencyOverrides {bLimitToOwner=true, maxCount=1, resolutionRule=StopOldest}
}
