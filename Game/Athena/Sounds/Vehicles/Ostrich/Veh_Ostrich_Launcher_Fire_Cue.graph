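# Sound cue for the Ostrich vehicle's launcher fire, layering close, medium and distant recordings.
# Each layer picks a random wave, passes through a modulator and uses its own attenuation.
# The close layer routes to the third-person fire sound class; medium and distant are mixed into the distant fire class at half input volume.
# Concurrency allows two instances per owner, stopping the oldest.
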SoundCue /Game/Athena/Sounds/Vehicles/Ostrich/Veh_Ostrich_Launcher_Fire_Cue
root=Mixer_3
nodes [
 {
  id=Attenuation_0
  children [Modulator_0]
  properties {attenuationSettings=&Fort_Weapon_Projectile_Medium_Mono_Athena_Att}
 }
 {id=Attenuation_11, children [Modulator_7], properties {attenuationSettings=/Game/Sounds/Attenuation/Fort_Weapon_Projectile_Distant_StW_Att}}
 {id=Attenuation_9, children [Modulator_6], properties {attenuationSettings=&Fort_Weapon_Projectile_Close_Athena_Att}}
 {id=Mixer_0, children [Attenuation_0, Attenuation_11], properties {inputVolume [1,1]}}
 {id=Mixer_3, children [SoundClass_7,SoundClass_6], properties {inputVolume [1,0.5]}}
 {id=Modulator_0, children [Random_1], properties {volumeMax=2.05, volumeMin=1.95}}
 {id=Modulator_6, children [Random_0], properties {pitchMax=1, pitchMin=1}}
 {id=Modulator_7, children [Random_2], properties {pitchMax=1.1, volumeMax=2.05, volumeMin=1.95}}
 {id=Random_0, children [WavePlayer_5,WavePlayer_6,WavePlayer_7,WavePlayer_8], properties {weights [1,1,1,1]}}
 {id=Random_1, children [WavePlayer_9, WavePlayer_10, WavePlayer_11, WavePlayer_18], properties {weights [1,1,1,1]}}
 {id=Random_2, children [WavePlayer_19, WavePlayer_20, WavePlayer_21], properties {weights [1,1,1]}}
 {id=SoundClass_6, children [Mixer_0], properties {soundClassOverride=&Fort_WeaponRanged_Fire_Distant}}
 {
  id=SoundClass_7
  children [Attenuation_9]
  properties {soundClassOverride=&Fort_WeaponRanged_Fire_3P_PSM}
 }
 {id=WavePlayer_10, properties {soundWaveAssetPtr=&/Veh_Ostrich_Launcher_Fire_Med_02}}
 {id=WavePlayer_11, properties {soundWaveAssetPtr=&/Veh_Ostrich_Launcher_Fire_Med_03}}
 {id=WavePlayer_18, properties {soundWaveAssetPtr=&/Veh_Ostrich_Launcher_Fire_Med_04}}
 {id=WavePlayer_19, properties {soundWaveAssetPtr=&/Veh_Ostrich_Launcher_Fire_Dist_01}}
 {id=WavePlayer_20, properties {soundWaveAssetPtr=&/Veh_Ostrich_Launcher_Fire_Dist_02}}
 {id=WavePlayer_21, properties {soundWaveAssetPtr=&/Veh_Ostrich_Launcher_Fire_Dist_03}}
 {id=WavePlayer_5, properties {soundWaveAssetPtr=&/Veh_Ostrich_Launcher_Fire_Close_01}}
 {id=WavePlayer_6, properties {soundWaveAssetPtr=&/Veh_Ostrich_Launcher_Fire_Close_02}}
 {id=WavePlayer_7, properties {soundWaveAssetPtr=&/Veh_Ostrich_Launcher_Fire_Close_03}}
 {id=WavePlayer_8, properties {soundWaveAssetPtr=&/Veh_Ostrich_Launcher_Fire_Close_04}}
 {id=WavePlayer_9, properties {soundWaveAssetPtr=&/Veh_Ostrich_Launcher_Fire_Med_01}}
]
properties {
 bOverrideConcurrency=true
 maxDistance=20000
 soundClassObject=&Fort_WeaponRanged_Fire_3P_PSM
 volumeMultiplier=0.6
 attenuationOverrides {falloffDistance=5000, attenuationShapeExtents {x=1000}}
 concurrencyOverrides {bLimitToOwner=true, maxCount=2, resolutionRule=StopOldest, volumeScale=0.5, volumeScaleAttackTime=0.05}
}
