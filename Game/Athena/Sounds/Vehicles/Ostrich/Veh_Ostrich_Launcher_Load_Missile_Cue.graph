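# Sound cue for the Ostrich vehicle's missile launcher loading a missile.
# A single wave player plays one missile load sample directly as the root, with no randomization or modulation.
# It uses the Ostrich close movement attenuation, routes to the weapon foley sound class, and plays at slightly reduced volume.
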SoundCue /Game/Athena/Sounds/Vehicles/Ostrich/Veh_Ostrich_Launcher_Load_Missile_Cue
root=WavePlayer_1
nodes [{id=WavePlayer_1, properties {soundWaveAssetPtr=&/Veh_Ostrich_Launcher_Missile_Load_01}}]
properties {
 attenuationSettings=/Game/Sounds/Attenuation/Fort_Ostrich_Movement_Close_Att
 maxDistance=6512
 soundClassObject=&Fort_WeaponFoley
 volumeMultiplier=0.8
}
