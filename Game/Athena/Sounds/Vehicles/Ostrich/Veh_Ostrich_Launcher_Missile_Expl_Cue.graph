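# Sound cue for the explosion of the Ostrich vehicle's launcher missile.
# A mixer layers close, medium and distant explosions, each with its own attenuation and sound class node, plus a half-volume close tail using missile in-air attenuation.
# Each layer picks a random wave variant, and modulators vary the three explosion layers, with explicit pitch ranges on medium and distant.
# An explosion concurrency set and a concurrency override are assigned to limit overlapping instances.
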SoundCue /Game/Athena/Sounds/Vehicles/Ostrich/Veh_Ostrich_Launcher_Missile_Expl_Cue
root=Mixer_0
nodes [
 {
  id=Attenuation_0
  children [Modulator_1]
  properties {
   attenuationSettings=/Game/Sounds/Attenuation/Fort_Explosion_Distant_Athena_Att
   attenuationOverrides {falloffDistance=3000, attenuationShapeExtents {x=1000}}
  }
 }
 {
  id=Attenuation_1
  children [Random_8]
  properties {attenuationSettings=/Game/Sounds/Attenuation/Fort_Weapon_Projectile_Ostrich_Missile_InAir}
 }
 {
  id=Attenuation_4
  children [Random_3]
  properties {attenuationSettings=/Game/Sounds/Attenuation/Fort_Explosion_Close_Athena_Att, attenuationOverrides {falloffDistance=1600}}
 }
 {
  id=Attenuation_5
  children [Modulator_5]
  properties {
   attenuationSettings=/Game/Sounds/Attenuation/Fort_Explosion_Medium_Athena_Att
   attenuationOverrides {falloffDistance=3000, attenuationShapeExtents {x=1000}}
  }
 }
 {id=Mixer_0, children [Mixer_4, Attenuation_1], properties {inputVolume [1,0.5]}}
 {id=Mixer_4, children [SoundClass_3,SoundClass_2,SoundClass_0], properties {inputVolume [1,0.75,0.75]}}
 {id=Modulator_0, children [Attenuation_4]}
 {id=Modulator_1, children [Random_5], properties {pitchMax=1.1, pitchMin=0.9}}
 {id=Modulator_5, children [Random_4], properties {pitchMax=1.1, pitchMin=0.9}}
 {id=Random_3, children [WavePlayer_2,WavePlayer_3,WavePlayer_5,WavePlayer_6,WavePlayer_7,WavePlayer_8], properties {weights [1,1,1,1,1,1]}}
 {
  id=Random_4
  children [WavePlayer_9, WavePlayer_10, WavePlayer_11, WavePlayer_18, WavePlayer_19, WavePlayer_20]
  properties {weights [1,1,1,1,1,1]}
 }
 {
  id=Random_5
  children [WavePlayer_21, WavePlayer_22, WavePlayer_23, WavePlayer_24, WavePlayer_25, WavePlayer_26]
  properties {weights [1,1,1,1,1,1]}
 }
 {id=Random_8, children [WavePlayer_40, WavePlayer_41, WavePlayer_42, WavePlayer_43, WavePlayer_44], properties {weights [1,1,1,1,1]}}
 {
  id=SoundClass_0
  children [Attenuation_0]
  properties {soundClassOverride=&Fort_WeaponRanged_Fire_Distant}
 }
 {
  id=SoundClass_2
  children [Attenuation_5]
  properties {soundClassOverride=&Fort_WeaponRanged_Fire_Distant}
 }
 {id=SoundClass_3, children [Modulator_0], properties {soundClassOverride=&Fort_WeaponRanged_Fire_3P_PSM}}
 {id=WavePlayer_10, properties {soundWaveAssetPtr=&/Veh_Ostrich_Launcher_Expl_Med_02}}
 {id=WavePlayer_11, properties {soundWaveAssetPtr=&/Veh_Ostrich_Launcher_Expl_Med_03}}
 {id=WavePlayer_18, properties {soundWaveAssetPtr=&/Veh_Ostrich_Launcher_Expl_Med_04}}
 {id=WavePlayer_19, properties {soundWaveAssetPtr=&/Veh_Ostrich_Launcher_Expl_Med_05}}
 {id=WavePlayer_2, properties {soundWaveAssetPtr=&/Veh_Ostrich_Launcher_Expl_Close_01}}
 {id=WavePlayer_20, properties {soundWaveAssetPtr=&/Veh_Ostrich_Launcher_Expl_Med_06}}
 {id=WavePlayer_21, properties {soundWaveAssetPtr=&/Veh_Ostrich_Launcher_Expl_Dist_01}}
 {id=WavePlayer_22, properties {soundWaveAssetPtr=&/Veh_Ostrich_Launcher_Expl_Dist_02}}
 {id=WavePlayer_23, properties {soundWaveAssetPtr=&/Veh_Ostrich_Launcher_Expl_Dist_03}}
 {id=WavePlayer_24, properties {soundWaveAssetPtr=&/Veh_Ostrich_Launcher_Expl_Dist_04}}
 {id=WavePlayer_25, properties {soundWaveAssetPtr=&/Veh_Ostrich_Launcher_Expl_Dist_05}}
 {id=WavePlayer_26, properties {soundWaveAssetPtr=&/Veh_Ostrich_Launcher_Expl_Dist_06}}
 {id=WavePlayer_3, properties {soundWaveAssetPtr=&/Veh_Ostrich_Launcher_Expl_Close_02}}
 {id=WavePlayer_40, properties {soundWaveAssetPtr=&/Veh_Ostrich_Launcher_Expl_Tail_Close_01}}
 {id=WavePlayer_41, properties {soundWaveAssetPtr=&/Veh_Ostrich_Launcher_Expl_Tail_Close_02}}
 {id=WavePlayer_42, properties {soundWaveAssetPtr=&/Veh_Ostrich_Launcher_Expl_Tail_Close_03}}
 {id=WavePlayer_43, properties {soundWaveAssetPtr=&/Veh_Ostrich_Launcher_Expl_Tail_Close_04}}
 {id=WavePlayer_44, properties {soundWaveAssetPtr=&/Veh_Ostrich_Launcher_Expl_Tail_Close_05}}
 {id=WavePlayer_5, properties {soundWaveAssetPtr=&/Veh_Ostrich_Launcher_Expl_Close_03}}
 {id=WavePlayer_6, properties {soundWaveAssetPtr=&/Veh_Ostrich_Launcher_Expl_Close_04}}
 {id=WavePlayer_7, properties {soundWaveAssetPtr=&/Veh_Ostrich_Launcher_Expl_Close_05}}
 {id=WavePlayer_8, properties {soundWaveAssetPtr=&/Veh_Ostrich_Launcher_Expl_Close_06}}
 {id=WavePlayer_9, properties {soundWaveAssetPtr=&/Veh_Ostrich_Launcher_Expl_Med_01}}
]
properties {
 bOverrideConcurrency=true
 maxDistance=27500
 soundClassObject=&Fort_WeaponRanged_Fire_3P_PSM
 volumeMultiplier=0.8
 concurrencyOverrides {maxCount=3}
 concurrencySet [/Game/Sounds/Concurrency_Rules/Weapon_Explosion_Conc]
}
