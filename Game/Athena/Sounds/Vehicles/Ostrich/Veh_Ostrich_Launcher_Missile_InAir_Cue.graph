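# Sound cue for the in-flight sound of missiles fired by the Ostrich vehicle's launcher.
# It randomly picks one of seven equally weighted missile tail waves, varies it through a modulator, routes it through a distant ranged-weapon-fire sound class node and applies smoothed, intensified Doppler at the root.
# A projectile-specific attenuation asset spatializes it, and a concurrency override caps simultaneous instances per owner.
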SoundCue /Game/Athena/Sounds/Vehicles/Ostrich/Veh_Ostrich_Launcher_Missile_InAir_Cue
root=Doppler_1
nodes [
 {id=Doppler_1, children [SoundClass_6], properties {bUseSmoothing=true, dopplerIntensity=3}}
 {id=Modulator_7, children [Random_0], properties {pitchMax=1.1}}
 {
  id=Random_0
  children [WavePlayer_2,WavePlayer_3,WavePlayer_4,WavePlayer_5,WavePlayer_6,WavePlayer_7,WavePlayer_8]
  properties {weights [1,1,1,1,1,1,1]}
 }
 {
  id=SoundClass_6
  children [Modulator_7]
  properties {soundClassOverride=&Fort_WeaponRanged_Fire_Distant}
 }
 {id=WavePlayer_2, properties {soundWaveAssetPtr=&/Veh_Ostrich_Launcher_Missile_Tail_01}}
 {id=WavePlayer_3, properties {soundWaveAssetPtr=&/Veh_Ostrich_Launcher_Missile_Tail_02}}
 {id=WavePlayer_4, properties {soundWaveAssetPtr=&/Veh_Ostrich_Launcher_Missile_Tail_03}}
 {id=WavePlayer_5, properties {soundWaveAssetPtr=&/Veh_Ostrich_Launcher_Missile_Tail_04}}
 {id=WavePlayer_6, properties {soundWaveAssetPtr=&/Veh_Ostrich_Launcher_Missile_Tail_05}}
 {id=WavePlayer_7, properties {soundWaveAssetPtr=&/Veh_Ostrich_Launcher_Missile_Tail_06}}
 {id=WavePlayer_8, properties {soundWaveAssetPtr=&/Veh_Ostrich_Launcher_Missile_Tail_07}}
]
properties {
 attenuationSettings=/Game/Sounds/Attenuation/Fort_Weapon_Projectile_Ostrich_Missile_InAir
 bOverrideConcurrency=true
 maxDistance=10512
 soundClassObject=&Fort_WeaponRanged_Fire_3P_PSM
 volumeMultiplier=0.4
 attenuationOverrides {falloffDistance=5000, attenuationShapeExtents {x=1000}}
 concurrencyOverrides {bLimitToOwner=true, maxCount=3, volumeScale=0.5}
}
