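# Sound cue for the Ostrich mech vehicle's large rotation servo foley in Athena.
# A Random node picks one of four equally weighted large servo waves, and a Modulator varies its pitch and volume for variety.
# It plays boosted in the player foley sound class with close movement attenuation and references the local footstep concurrency set, alongside per-owner, stop-oldest override settings.
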SoundCue /Game/Athena/Sounds/Vehicles/Ostrich/Veh_Ostrich_RotateLarge_Cue
root=Modulator_0
nodes [
 {id=Modulator_0, children [Random_0], properties {pitchMin=0.9, volumeMin=0.8}}
 {id=Random_0, children [WavePlayer_0,WavePlayer_1,WavePlayer_2,WavePlayer_3], properties {weights [1,1,1,1]}}
 {
  id=WavePlayer_0
  properties {soundWaveAssetPtr=/Game/Athena/Sounds/Vehicles/Ostrich/ProceduralFoley/Mech_Vehicle_Prodedural_Servo_SetB_Large_04}
 }
 {
  id=WavePlayer_1
  properties {soundWaveAssetPtr=/Game/Athena/Sounds/Vehicles/Ostrich/ProceduralFoley/Mech_Vehicle_Prodedural_Servo_SetB_Large_01}
 }
 {
  id=WavePlayer_2
  properties {soundWaveAssetPtr=/Game/Athena/Sounds/Vehicles/Ostrich/ProceduralFoley/Mech_Vehicle_Prodedural_Servo_SetB_Large_02}
 }
 {
  id=WavePlayer_3
  properties {soundWaveAssetPtr=/Game/Athena/Sounds/Vehicles/Ostrich/ProceduralFoley/Mech_Vehicle_Prodedural_Servo_SetB_Large_03}
 }
]
properties {
 attenuationSettings=/Game/Sounds/Attenuation/Fort_Ostrich_Movement_Close_Att
 maxDistance=6512
 soundClassObject=&Fort_Player_Foley_Athena
 volumeMultiplier=2
 concurrencyOverrides {bLimitToOwner=true, maxCount=1, resolutionRule=StopOldest}
 concurrencySet [&Athena_Footstep_Local_Conc]
}
