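# Sound cue for the looping broken-shield sound on the Ostrich vehicle.
# A single looping wave player plays the shield damage loop wave at boosted volume, with no random or mixing nodes.
# It uses the Ostrich close movement attenuation and the Athena player foley sound class.
# Concurrency overrides specify one instance per owner, stopping the oldest.
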SoundCue /Game/Athena/Sounds/Vehicles/Ostrich/Veh_Ostrich_Shield_Broken_Loop_Cue
root=WavePlayer_4
nodes [{id=WavePlayer_4, properties {bLooping=true, soundWaveAssetPtr=&/Veh_Ostrich_Shield_Damage_Loop}}]
properties {
 attenuationSettings=/Game/Sounds/Attenuation/Fort_Ostrich_Movement_Close_Att
 maxDistance=6512
 soundClassObject=&Fort_Player_Foley_Athena
 volumeMultiplier=2
 concurrencyOverrides {bLimitToOwner=true, maxCount=1, resolutionRule=StopOldest}
 concurrencySet [null]
}
