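# Sound cue for impacts on the Ostrich vehicle's shield.
# A Random node picks one of five equally weighted impact waves, and a Modulator varies its pitch and volume for variety.
# It uses close Ostrich movement attenuation and a 2D hit-notification instigator sound class.
# A referenced concurrency asset, named for stop-oldest, owner-limited behavior, governs overlapping plays.
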SoundCue /Game/Athena/Sounds/Vehicles/Ostrich/Veh_Ostrich_Shield_Impact_Cue
root=Modulator_0
nodes [
 {id=Modulator_0, children [Random_0], properties {pitchMin=0.9, volumeMin=0.8}}
 {id=Random_0, children [WavePlayer_0,WavePlayer_1,WavePlayer_2,WavePlayer_3,WavePlayer_5], properties {weights [1,1,1,1,1]}}
 {id=WavePlayer_0, properties {soundWaveAssetPtr=&/Veh_Ostrich_Shield_Impact_04}}
 {id=WavePlayer_1, properties {soundWaveAssetPtr=&/Veh_Ostrich_Shield_Impact_05}}
 {id=WavePlayer_2, properties {soundWaveAssetPtr=&/Veh_Ostrich_Shield_Impact_01}}
 {id=WavePlayer_3, properties {soundWaveAssetPtr=&/Veh_Ostrich_Shield_Impact_02}}
 {id=WavePlayer_5, properties {soundWaveAssetPtr=&/Veh_Ostrich_Shield_Impact_03}}
]
properties {
 attenuationSettings=/Game/Sounds/Attenuation/Fort_Ostrich_Movement_Close_Att
 maxDistance=6512
 soundClassObject=&Fort_2DHitNotifications_Instigate_PSM
 volumeMultiplier=1
 concurrencyOverrides {bLimitToOwner=true, maxCount=1, resolutionRule=StopOldest}
 concurrencySet [&/Veh_Ostrich_StopOldestLimitToOwnerFFS_Conc]
}
