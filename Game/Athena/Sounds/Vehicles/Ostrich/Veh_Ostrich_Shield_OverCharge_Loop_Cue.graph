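# Sound cue that plays the looping shield overcharge sound for the Ostrich vehicle.
# A single looping wave player feeds the output at half volume, using the Ostrich close-movement attenuation.
# It routes to a 2D hit notification instigator sound class despite having attenuation set.
# Concurrency overrides specify one instance per owner, stopping the oldest, though the concurrency set holds a null entry.
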SoundCue /Game/Athena/Sounds/Vehicles/Ostrich/Veh_Ostrich_Shield_OverCharge_Loop_Cue
root=WavePlayer_1
nodes [{id=WavePlayer_1, properties {bLooping=true, soundWaveAssetPtr=&/Veh_Ostrich_Shield_OverCharge_Lp_01}}]
properties {
 attenuationSettings=/Game/Sounds/Attenuation/Fort_Ostrich_Movement_Close_Att
 maxDistance=6512
 soundClassObject=&Fort_2DHitNotifications_Instigate_PSM
 volumeMultiplier=0.5
 concurrencyOverrides {bLimitToOwner=true, maxCount=1, resolutionRule=StopOldest}
 concurrencySet [null]
}
