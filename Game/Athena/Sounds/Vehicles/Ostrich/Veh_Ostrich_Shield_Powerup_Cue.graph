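# Sound cue for the Ostrich vehicle's shield powerup, playing a single shield-raise wave.
# A Modulator feeds from the wave player to randomize pitch and volume on each play, with a reduced overall volume multiplier.
# It uses close-range Ostrich movement attenuation and an instigator hit-notification sound class.
# Its assigned concurrency set is named for stopping the oldest instance, limited to the owner.
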SoundCue /Game/Athena/Sounds/Vehicles/Ostrich/Veh_Ostrich_Shield_Powerup_Cue
root=Modulator_0
nodes [
 {id=Modulator_0, children [WavePlayer_4], properties {pitchMin=0.9, volumeMin=0.8}}
 {id=WavePlayer_4, properties {soundWaveAssetPtr=&/Veh_Ostrich_Shield_Raise_01}}
]
properties {
 attenuationSettings=/Game/Sounds/Attenuation/Fort_Ostrich_Movement_Close_Att
 maxDistance=6512
 soundClassObject=&Fort_2DHitNotifications_Instigate_PSM
 volumeMultiplier=0.3
 concurrencyOverrides {bLimitToOwner=true, maxCount=1, resolutionRule=StopOldest}
 concurrencySet [&/Veh_Ostrich_StopOldestLimitToOwnerFFS_Conc]
}
