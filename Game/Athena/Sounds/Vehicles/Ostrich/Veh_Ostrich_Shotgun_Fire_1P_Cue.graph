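# Sound cue for the first-person shotgun fire of the Ostrich vehicle.
# A mixer layers one of four random close-range fire shots over a much quieter shell casing layer.
# The casings are random picks borrowed from the M249 SAW sounds, delayed one second and pitched down.
# It uses first-person weapon attenuation, sound class and gunfire concurrency settings.
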SoundCue /Game/Athena/Sounds/Vehicles/Ostrich/Veh_Ostrich_Shotgun_Fire_1P_Cue
root=Mixer_1
nodes [
 {id=Delay_0, children [Random_0], properties {delayMax=1, delayMin=1}}
 {id=Mixer_1, children [Random_3,Modulator_0], properties {inputVolume [1,0.1]}}
 {id=Modulator_0, children [Delay_0], properties {pitchMax=0.8, pitchMin=0.8}}
 {id=Random_0, children [WavePlayer_0,WavePlayer_1,WavePlayer_2,WavePlayer_7], properties {weights [1,1,1,1]}}
 {id=Random_3, children [WavePlayer_3,WavePlayer_4,WavePlayer_5,WavePlayer_6], properties {weights [1,1,1,1]}}
 {id=WavePlayer_0, properties {soundWaveAssetPtr=/Game/Sounds/Fort_Weapons/AutoRifle/M249_Saw/M249_Casings_04}}
 {id=WavePlayer_1, properties {soundWaveAssetPtr=/Game/Sounds/Fort_Weapons/AutoRifle/M249_Saw/M249_Casings_01}}
 {id=WavePlayer_2, properties {soundWaveAssetPtr=/Game/Sounds/Fort_Weapons/AutoRifle/M249_Saw/M249_Casings_02}}
 {id=WavePlayer_3, properties {soundWaveAssetPtr=&/Veh_Ostrich_Shotgun_Fire_Close_01}}
 {id=WavePlayer_4, properties {soundWaveAssetPtr=&/Veh_Ostrich_Shotgun_Fire_Close_02}}
 {id=WavePlayer_5, properties {soundWaveAssetPtr=&/Veh_Ostrich_Shotgun_Fire_Close_03}}
 {id=WavePlayer_6, properties {soundWaveAssetPtr=&/Veh_Ostrich_Shotgun_Fire_Close_04}}
 {id=WavePlayer_7, properties {soundWaveAssetPtr=/Game/Sounds/Fort_Weapons/AutoRifle/M249_Saw/M249_Casings_03}}
]
properties {
 attenuationSettings=&Fort_Weapon_Projectile_FirstPerson_Athena_Att
 maxDistance=2097152
 soundClassObject=&Fort_WeaponRanged_Fire_1P_PSM
 attenuationOverrides {falloffDistance=5000, attenuationShapeExtents {x=1000}}
 concurrencyOverrides {bLimitToOwner=true, maxCount=3}
 concurrencySet [&Weapon_Gunfire_FirstPerson_Conc]
}
