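# Sound cue for the Ostrich vehicle's shotgun fire, layering close, medium and distant recordings.
# Each layer randomly picks one of four waves, then runs through a volume modulator (the distant one also varies pitch) and its own range attenuation.
# Sound class overrides route close to third-person weapon fire and the other two to distant fire, before a mixer blends them with both far layers halved.
# Overridden concurrency caps simultaneous instances per owner.
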SoundCue /Game/Athena/Sounds/Vehicles/Ostrich/Veh_Ostrich_Shotgun_Fire_Cue
root=Mixer_3
nodes [
 {
  id=Attenuation_1
  children [Modulator_1]
  properties {attenuationSettings=&Fort_Weapon_Projectile_Medium_Mono_Athena_Att}
 }
 {id=Attenuation_2, children [Modulator_7], properties {attenuationSettings=&Fort_Weapon_Projectile_Close_Athena_Att}}
 {
  id=Attenuation_3
  children [Modulator_6]
  properties {attenuationSettings=&Fort_Weapon_Projectile_Distant_Mono_Athena_Att}
 }
 {id=Mixer_3, children [SoundClass_2,SoundClass_6,SoundClass_3], properties {inputVolume [1,0.5,0.5]}}
 {id=Modulator_1, children [Random_8], properties {volumeMax=2, volumeMin=2}}
 {id=Modulator_6, children [Random_6], properties {pitchMax=1, pitchMin=0.9, volumeMax=2, volumeMin=2}}
 {id=Modulator_7, children [Random_2], properties {volumeMax=1.3, volumeMin=1.25}}
 {id=Random_2, children [WavePlayer_0,WavePlayer_1,WavePlayer_2,WavePlayer_3], properties {weights [1,1,1,1]}}
 {id=Random_6, children [WavePlayer_8, WavePlayer_9, WavePlayer_19, WavePlayer_20], properties {weights [1,1,1,1]}}
 {id=Random_8, children [WavePlayer_4,WavePlayer_5,WavePlayer_6,WavePlayer_7], properties {weights [1,1,1,1]}}
 {
  id=SoundClass_2
  children [Attenuation_2]
  properties {soundClassOverride=&Fort_WeaponRanged_Fire_3P_PSM}
 }
 {
  id=SoundClass_3
  children [Attenuation_3]
  properties {soundClassOverride=&Fort_WeaponRanged_Fire_Distant}
 }
 {
  id=SoundClass_6
  children [Attenuation_1]
  properties {soundClassOverride=&Fort_WeaponRanged_Fire_Distant}
 }
 {id=WavePlayer_0, properties {soundWaveAssetPtr=&/Veh_Ostrich_Shotgun_Fire_Close_01}}
 {id=WavePlayer_1, properties {soundWaveAssetPtr=&/Veh_Ostrich_Shotgun_Fire_Close_02}}
 {id=WavePlayer_19, properties {soundWaveAssetPtr=&/Veh_Ostrich_Shotgun_Fire_Dist_03}}
 {id=WavePlayer_2, properties {soundWaveAssetPtr=&/Veh_Ostrich_Shotgun_Fire_Close_03}}
 {id=WavePlayer_20, properties {soundWaveAssetPtr=&/Veh_Ostrich_Shotgun_Fire_Dist_04}}
 {id=WavePlayer_3, properties {soundWaveAssetPtr=&/Veh_Ostrich_Shotgun_Fire_Close_04}}
 {id=WavePlayer_4, properties {soundWaveAssetPtr=&/Veh_Ostrich_Shotgun_Fire_Med_01}}
 {id=WavePlayer_5, properties {soundWaveAssetPtr=&/Veh_Ostrich_Shotgun_Fire_Med_02}}
 {id=WavePlayer_6, properties {soundWaveAssetPtr=&/Veh_Ostrich_Shotgun_Fire_Med_03}}
 {id=WavePlayer_7, properties {soundWaveAssetPtr=&/Veh_Ostrich_Shotgun_Fire_Med_04}}
 {id=WavePlayer_8, properties {soundWaveAssetPtr=&/Veh_Ostrich_Shotgun_Fire_Dist_01}}
 {id=WavePlayer_9, properties {soundWaveAssetPtr=&/Veh_Ostrich_Shotgun_Fire_Dist_02}}
]
properties {
 bOverrideConcurrency=true
 maxDistance=27500
 soundClassObject=&Fort_WeaponRanged_Fire_3P_PSM
 attenuationOverrides {falloffDistance=5000, attenuationShapeExtents {x=1000}}
 concurrencyOverrides {bLimitToOwner=true, maxCount=3}
}
