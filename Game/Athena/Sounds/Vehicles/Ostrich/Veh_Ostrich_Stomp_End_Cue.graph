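# Sound cue for the servo sound that ends a stomp by the Ostrich vehicle.
# A Random node picks between two equally weighted servo end waves.
# It plays at reduced volume through the distant Ostrich movement attenuation and the third-person enemy foley sound class.
# A cue-level concurrency override blocks new instances once the limit is reached, rather than stopping ones already playing.
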SoundCue /Game/Athena/Sounds/Vehicles/Ostrich/Veh_Ostrich_Stomp_End_Cue
root=Random_0
nodes [
 {id=Random_0, children [WavePlayer_0,WavePlayer_1], properties {weights [1,1]}}
 {id=WavePlayer_0, properties {soundWaveAssetPtr=&/Veh_Ostrich_Stomp_Servo_End_01}}
 {id=WavePlayer_1, properties {soundWaveAssetPtr=&/Veh_Ostrich_Stomp_Servo_End_02}}
]
properties {
 attenuationSettings=/Game/Sounds/Attenuation/Fort_Ostrich_Movement_Distant_Att
 bOverrideConcurrency=true
 maxDistance=15000
 soundClassObject=&Fort_Player_Foley_3P_Enemy_Athena
 volumeMultiplier=0.5
 attenuationOverrides {falloffDistance=5000, reverbWetLevelMax=0.7, reverbWetLevelMin=0.2, attenuationShapeExtents {x=1000}}
 concurrencyOverrides {resolutionRule=PreventNew}
}
