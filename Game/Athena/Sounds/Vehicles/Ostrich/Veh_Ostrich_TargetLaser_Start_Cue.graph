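# Sound cue for the start of the Ostrich vehicle's targeting laser, including the looping beam sound that follows.
# Mixes three separately attenuated layers.
# The distant-movement attenuation layer pairs a random start one-shot with a loop under default modulation.
# The targeting-laser and close attenuation layers each loop a pitch-locked wave, the close one reusing a Storm King beam, with enveloper volume ramps over the first quarter second.
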SoundCue /Game/Athena/Sounds/Vehicles/Ostrich/Veh_Ostrich_TargetLaser_Start_Cue
root=Mixer_1
nodes [
 {
  id=Attenuation_0
  children [Enveloper_0]
  properties {attenuationSettings=&/Veh_Ostrich_TargetingLaser_Att}
 }
 {id=Attenuation_1, children [Modulator_0], properties {attenuationSettings=/Game/Sounds/Attenuation/Fort_Ostrich_Movement_Distant_Att}}
 {
  id=Attenuation_2
  children [Enveloper_1]
  properties {attenuationSettings=&/Veh_Ostrich_TargetingLaser_Close_Att}
 }
 {
  id=Enveloper_0
  children [Modulator_1]
  properties {
   pitchCurve {editorCurveData {keys [{time=0}, {arriveTangent=0.0936, leaveTangent=0.0936, time=3}]}}
   volumeCurve {editorCurveData {keys [<at 0 0.1>,<at 0.25 1>]}}
  }
 }
 {
  id=Enveloper_1
  children [Modulator_2]
  properties {
   pitchCurve {editorCurveData {keys [{time=0}, {arriveTangent=0.0936, leaveTangent=0.0936, time=3}]}}
   volumeCurve {editorCurveData {keys [<at 0 0.1>,<at 0.25 1>]}}
  }
 }
 {id=Mixer_0, children [Random_0,WavePlayer_3], properties {inputVolume [1,1]}}
 {id=Mixer_1, children [Attenuation_1, Attenuation_0, Attenuation_2], properties {inputVolume [1.5,1,0.75]}}
 {id=Modulator_0, children [Mixer_0]}
 {id=Modulator_1, children [WavePlayer_4], properties {pitchMax=1, pitchMin=1}}
 {id=Modulator_2, children [WavePlayer_1], properties {pitchMax=1, pitchMin=1}}
 {id=Random_0, children [WavePlayer_0,WavePlayer_2], properties {weights [1,1]}}
 {id=WavePlayer_0, properties {soundWaveAssetPtr=&/Veh_Ostrich_TargetLaser_Start_01}}
 {id=WavePlayer_1, properties {bLooping=true, soundWaveAssetPtr=/Game/Sounds/Fort_Monster_Storm_King/SFX/StormKing_BeamAttack_Beam_02_Loop}}
 {id=WavePlayer_2, properties {soundWaveAssetPtr=&/Veh_Ostrich_TargetLaser_Start_02}}
 {id=WavePlayer_3, properties {bLooping=true, soundWaveAssetPtr=&/Veh_Ostrich_TargetLaser_Loop_01}}
 {id=WavePlayer_4, properties {bLooping=true, soundWaveAssetPtr=&/Veh_Ostrich_TargetLaser_Loop_03}}
]
properties {maxDistance=22500, soundClassObject=&Fort_WeaponRanged_Fire_3P_PSM, volumeMultiplier=0.8}
