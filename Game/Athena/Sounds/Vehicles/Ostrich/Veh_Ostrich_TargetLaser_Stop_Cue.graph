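# Sound cue for the Ostrich vehicle's target laser stop sound.
# A random node picks one of two equally weighted stop wave variants, then a modulator at the root applies its default variation.
# The cue plays slightly reduced in volume, using the distant Ostrich movement attenuation and a third-person ranged weapon fire sound class.
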SoundCue /Game/Athena/Sounds/Vehicles/Ostrich/Veh_Ostrich_TargetLaser_Stop_Cue
root=Modulator_0
nodes [
 {id=Modulator_0, children [Random_1]}
 {id=Random_1, children [WavePlayer_1,WavePlayer_4], properties {weights [1,1]}}
 {id=WavePlayer_1, properties {soundWaveAssetPtr=&/Veh_Ostrich_TargetLaser_Stop_02}}
 {id=WavePlayer_4, properties {soundWaveAssetPtr=&/Veh_Ostrich_TargetLaser_Stop_01}}
]
properties {
 attenuationSettings=/Game/Sounds/Attenuation/Fort_Ostrich_Movement_Distant_Att
 maxDistance=15000
 soundClassObject=&Fort_WeaponRanged_Fire_3P_PSM
 volumeMultiplier=0.8
}
